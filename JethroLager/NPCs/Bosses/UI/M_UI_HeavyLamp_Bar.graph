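# Translucent UI material for the HeavyLamp boss health bar, with a fill driven by the Progress parameter.
# MF_UI_4VertexPolygon builds the fill, outline and glow masks from vertex parameters, with adjustable scale and corner roundness.
# Horizontal UI UV is compared against Progress to form the fill split, a thin progress mark and a colored gradient at the edge.
# Named reroutes carry these masks into layered color and alpha blends, with an optional outline and overall opacity.
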Material /JethroLager/NPCs/Bosses/UI/M_UI_HeavyLamp_Bar
graph {
 properties {bAutomaticallySetUsageInEditor=false, blendMode=BLEND_Translucent, materialDomain=MD_UI, usedShadingModels=Default Lit}
 nodes [
  n1 = TextureCoordinate()
  n2 = ScalarParameter() {
   defaultValue=0.7
   desc="When progress value changes, it defines where change starts.\r\n0 - 1"
   group=Progress
   parameterName=ProgressStart
   sliderMax=1
   sortPriority=1
  }
  n3 = ScalarParameter() {
   defaultValue=0.3
   desc="When progress value changes, it defines where change ends.\r\n0 - 1"
   group=Progress
   parameterName=ProgressEnd
   sliderMax=1
   sortPriority=2
  }
  n4 = ScalarParameter() {defaultValue=0.5, group=Progress, parameterName=Progress, sliderMax=1}
  n5 = Clamp(None=Power(Base=LinearInterpolate(A=@n2, B=@n3, Alpha=@n4)) {constExponent=1})
  Add(A=@n1, B=@n5) {constB=0.5}
  n6 = ComponentMask(None=MaterialFunctionCall() {materialFunction=&GetUserInterfaceUV}.Normalized UV) {r=true}
  Clamp(None=@n6) {maxDefault=0.7, minDefault=0.3}
  n7 = NamedRerouteUsage()
  n8 = Max(
   A=Multiply(A=Power(Base=@n7)) {constB=2}
   B=LinearInterpolate(
    A=VectorParameter() {parameterName=EmptyColor, defaultValue=<rgba 0.01 0.0001 0.0005 1>}.RGB
    B=VectorParameter() {parameterName=FillColor, defaultValue=<rgba 1 0.0097 0.0513 1>}.RGB
    Alpha=NamedRerouteUsage()
   )
  )
  n9 = VectorParameter() {parameterName=EmptyColor, defaultValue=<rgba 0.01 0.0001 0.0005 1>}
  n10 = NamedRerouteUsage()
  LinearInterpolate(
   A=LinearInterpolate(
    A=LinearInterpolate(A=@n8, B=@n9.RGB, Alpha=@n10)
    B=VectorParameter() {parameterName=OutlineColor, defaultValue=<rgba 1 1 1 1>}.RGB
    Alpha=ScalarParameter() {parameterName=Show Outline}
   )
   B=VectorParameter() {parameterName=Progress Mark Color, defaultValue=<rgba 1 0.9 0.8 1>}.RGB
   Alpha=NamedRerouteUsage()
  )
  n11 = ScalarParameter() {
   defaultValue=0.7
   desc="When progress value changes, it defines where change starts.\r\n0 - 1"
   group=Progress
   parameterName=ProgressStart
   sliderMax=1
   sortPriority=1
  }
  n12 = ScalarParameter() {
   defaultValue=0.3
   desc="When progress value changes, it defines where change ends.\r\n0 - 1"
   group=Progress
   parameterName=ProgressEnd
   sliderMax=1
   sortPriority=2
  }
  n13 = ScalarParameter() {defaultValue=0.5, group=Progress, parameterName=Progress, sliderMax=1}
  LinearInterpolate(A=@n11, B=@n12, Alpha=OneMinus(None=@n13))
  n14 = MaterialFunctionCall() {materialFunction=&GetUserInterfaceUV}
  n15 = Subtract(A=ComponentMask(None=@n14.Normalized UV) {r=true}) {constB=0.99}
  n16 = MaterialFunctionCall(UV Channel=AppendVector(A=Add(A=@n15, B=@n13), B=ComponentMask(None=@n14.Normalized UV) {g=true})) {materialFunction=&LinearGradient}
  Multiply(A=Multiply(A=Power(Base=@n16.UGradient) {constExponent=0.5}, B=NamedRerouteUsage()), B=NamedRerouteUsage())
  Multiply(A=@n1) {constB=-1}
  n17 = NamedRerouteUsage()
  n18 = Multiply(A=NamedRerouteUsage(), B=NamedRerouteUsage())
  n19 = LinearInterpolate(A=NamedRerouteUsage(), B=@n18, Alpha=@n18)
  Multiply(
   A=LinearInterpolate(A=LinearInterpolate(A=@n17, B=@n19, Alpha=@n19), B=NamedRerouteUsage(), Alpha=ScalarParameter() {parameterName=Show Outline})
   B=ScalarParameter() {defaultValue=0.8, parameterName=Opacity}
  )
  NamedRerouteDeclaration() {name=alpha, nodeColor=<rgba 1 0 0.1875 1>}
  n20 = MaterialFunctionCall() {materialFunction=&GetUserInterfaceUV}
  n21 = Divide(
   A=Subtract(
    A=ScalarParameter() {defaultValue=0.5, group=Progress, parameterName=Progress, sliderMax=1}
    B=ComponentMask(None=@n20.Normalized UV) {r=true}
   )
   B=ScalarParameter() {defaultValue=0.05, parameterName=Progress Gradient width}
  )
  n22 = Add(A=Multiply(A=Saturate(None=@n21)) {constB=-1})
  n23 = NamedRerouteUsage()
  NamedRerouteDeclaration(None=Multiply(A=Multiply(A=@n22, B=@n23), B=VectorParameter() {parameterName=Gradient Mark Color, defaultValue=<rgba 1 0.5 0 1>}.RGB)) {name=Gradient Mark, nodeColor=<rgba 0 0 0 1>}
  n24 = MaterialFunctionCall(UVs=TextureCoordinate(), Texture Scale=ScalarParameter() {defaultValue=1, parameterName=Scale}) {materialFunction=&ScaleUVsByCenter}
  n25 = VectorParameter() {parameterName=Vert0, defaultValue=<rgba -0.5 -1 0.5 -1>}
  n26 = AppendVector(A=@n25.R, B=@n25.R)
  n27 = AppendVector(A=@n25.B, B=@n25.B)
  n28 = VectorParameter() {parameterName=Vert1, defaultValue=<rgba 1 1 -1 1>}
  n29 = AppendVector(A=@n28.R, B=@n28.R)
  n30 = AppendVector(A=@n28.B, B=@n28.B)
  n31 = Constant()
  n32 = Constant() {r=0.2}
  n33 = ScalarParameter() {defaultValue=-1, parameterName=Alignment}
  n34 = ScalarParameter() {parameterName=CornerRoundness}
  NamedRerouteDeclaration(None=Multiply(
    A=MaterialFunctionCall!(
     UVs=@n24.UVs
     vertex_0=@n26
     vertex_1=@n27
     vertex_2=@n29
     vertex_3=@n30
     Glow Min=@n31
     Glow Max=@n32
     OutlineThickness
     OutlineAlignment=@n33
     CornerRoundness=@n34
    ) {materialFunction=&/MF_UI_4VertexPolygon}.Fill
    B=NamedRerouteUsage()
   )) {name=Glow, nodeColor=<rgba 1 0 0.7969 1>}
  n35 = MaterialFunctionCall(
   UVs=MaterialFunctionCall(UVs=TextureCoordinate(), Texture Scale=ScalarParameter() {defaultValue=1, parameterName=Scale}) {materialFunction=&ScaleUVsByCenter}.UVs
   vertex_0=@n26
   vertex_1=@n27
   vertex_2=@n29
   vertex_3=@n30
   OutlineAlignment=ScalarParameter() {defaultValue=-1, parameterName=Alignment}
   CornerRoundness=ScalarParameter() {parameterName=CornerRoundness}
  ) {materialFunction=&MF_UI_4VertexPolygon}
  NamedRerouteDeclaration(None=@n35.Fill) {name=BossHealthBar_Fill, nodeColor=<rgba 0 0 0 1>}
  n36 = TextureCoordinate()
  n37 = ScalarParameter() {defaultValue=1, parameterName=OutlineScale}
  NamedRerouteDeclaration(None=MaterialFunctionCall(
    UVs=MaterialFunctionCall(UVs=@n36, Texture Scale=@n37) {materialFunction=&ScaleUVsByCenter}.UVs
    vertex_0=@n26
    vertex_1=@n27
    vertex_2=@n29
    vertex_3=@n30
    OutlineThickness=ScalarParameter() {defaultValue=0.02, parameterName=OutlineThickness}
    OutlineAlignment=ScalarParameter() {defaultValue=-1, parameterName=Alignment}
    CornerRoundness=ScalarParameter() {parameterName=CornerRoundness}
   ) {materialFunction=&MF_UI_4VertexPolygon}.Outline) {name=BossHealthBar_Outline, nodeColor=<rgba 0 0.1328 1 1>}
  NamedRerouteDeclaration() {name=Progress, nodeColor=<rgba 0 0 0 1>}
  n38 = Subtract(A=@n6, B=ScalarParameter() {defaultValue=0.5, group=Progress, parameterName=Progress, sliderMax=1})
  NamedRerouteDeclaration(None=OneMinus(None=Step(X=@n38) {constX=0})) {name=ProgressGradient, nodeColor=<rgba 0 0 0 1>}
  NamedRerouteDeclaration(None=@n35.SDF) {nodeColor=<rgba 0.9688 0 1 1>}
  n39 = ScalarParameter() {defaultValue=0.5, group=Progress, parameterName=Progress, sliderMax=1}
  n40 = ComponentMask(None=MaterialFunctionCall() {materialFunction=&GetUserInterfaceUV}.Normalized UV) {r=true}
  n41 = Step(Y=@n39, X=@n40)
  n42 = Step(Y=Add(A=@n39) {constB=0.003}, X=@n40)
  NamedRerouteDeclaration(None=LinearInterpolate(B=Subtract(A=@n41, B=@n42), Alpha=ScalarParameter() {defaultValue=1, parameterName=Show Progress Mark})) {name=Progress Mark, nodeColor=<rgba 0 0 0 1>}
  n43 = Subtract(A=@n1, B=@n5)
  Saturate(None=SmoothStep(
    Max=ScalarParameter() {group=Progress_Layout, parameterName=ProgressEdgeSharpness, sliderMax=1, sortPriority=2}
    Value=ComponentMask(None=@n43) {r=true}
   ) {constMax=0.01})
  ScalarParameter() {
   defaultValue=0.5
   desc="Property for animating progress value change (from ProgressStart to ProgressEnd)\r\n0 - 1"
   group=Progress
   parameterName=AnimateProgress
   sliderMax=1
   sortPriority=3
  }
  TextureCoordinate()
  TextureCoordinate()
  TextureCoordinate()
 ]
}
