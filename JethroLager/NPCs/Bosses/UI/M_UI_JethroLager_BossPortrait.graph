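# Translucent UI material for a JethroLager boss portrait with a colored hologram effect.
# A scalar parameter blends between the HeavyChandelier and BinGrass portraits, both scaled around their center.
# A softened triangle SDF, scaled by an opacity parameter, forms the hologram beam shape.
# Time-panned repeating scanlines and sine-based moving lines add animation, with named reroutes carrying each layer into gradient-masked lerps.
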Material /JethroLager/NPCs/Bosses/UI/M_UI_JethroLager_BossPortrait
graph {
 properties {bAutomaticallySetUsageInEditor=false, blendMode=BLEND_Translucent, materialDomain=MD_UI, usedShadingModels=Default Lit}
 nodes [
  Constant() {r=0.5}
  n1 = NamedRerouteUsage()
  n2 = MaterialFunctionCall(RotationAngle=Constant() {r=-0.25}) {materialFunction=&MF_UI_LinearGradient}
  n3 = NamedRerouteUsage()
  LinearInterpolate(A=Min(A=Multiply(A=@n1, B=@n2.Result), B=NamedRerouteUsage()), B=@n3, Alpha=@n3)
  LinearInterpolate(
   A=VectorParameter() {parameterName=Hologram Color, defaultValue=<rgba 0.8228 0.1521 0.1619 1>}.RGB
   B=NamedRerouteUsage()
   Alpha=NamedRerouteUsage()
  )
  NamedRerouteDeclaration() {name=MovingLines, nodeColor=<rgba 0 0 0 1>}
  n4 = TextureCoordinate()
  n5 = ComponentMask(None=@n4) {r=true}
  n6 = Add(A=ComponentMask(None=@n4) {g=true}, B=Constant() {r=-0.28}) {constB=0}
  n7 = MaterialFunctionCall!(
   UVs=MaterialFunctionCall(UVs=AppendVector(A=@n5, B=@n6), Texture Scale=Constant() {r=0.8}) {materialFunction=&ScaleUVsByCenter}.UVs
   Corner 01=Constant2Vector() {g=-1.4, r=-0.7}
   Corner 02=Constant2Vector() {g=-1.4, r=0.7}
   Corner 03=Constant2Vector() {g=0.05}
   Glow Min
   Glow Max
   Outline Thickness
  ) {materialFunction=/FortUILibrary/MaterialFunctions/Shapes/MF_UI_WIP_Triangle}
  n8 = OneMinus(None=Power(Base=SmoothStep(Value=@n7.SDF) {constMax=0.5}))
  n9 = ScalarParameter() {defaultValue=1, parameterName=Hologram Opacity}
  NamedRerouteDeclaration(None=Multiply(A=Multiply(A=@n8, B=@n9) {constB=0.5}, B=NamedRerouteUsage())) {name=Hologram, nodeColor=<rgba 0 0 0 1>}
  n10 = MaterialFunctionCall(Texture Scale=ScalarParameter() {defaultValue=0.9, parameterName=Portrait scale}) {materialFunction=&ScaleUVsByCenter}
  n11 = TextureSample(UVs=@n10.UVs) {samplerSource=SSM_Clamp_WorldGroupSettings, texture=/JethroLager/NPCs/Bosses/UI/Textures/T_UI_HeavyChandelier_Portrait}
  n12 = TextureSample(UVs=@n10.UVs) {samplerSource=SSM_Clamp_WorldGroupSettings, texture=/JethroLager/NPCs/Bosses/UI/Textures/T_UI_BinGrass_Portrait}
  n13 = ScalarParameter() {parameterName=ShowBinGrassPortrait}
  NamedRerouteDeclaration(None=LinearInterpolate(A=@n11.RGB, B=@n12.RGB, Alpha=@n13)) {name=Portrait, nodeColor=<rgba 0 0 0 1>}
  NamedRerouteDeclaration(None=LinearInterpolate(A=@n11.A, B=@n12.A, Alpha=@n13)) {name=PortraitAlpha, nodeColor=<rgba 0 0 0 1>}
  n14 = Constant() {r=0.3}
  n15 = Subtract(A=@n14) {constB=0.001}
  n16 = TextureCoordinate()
  n17 = Multiply(A=Time(), B=ScalarParameter() {defaultValue=0.1, parameterName=Scaline Speed})
  n18 = Multiply(
   A=ComponentMask(None=Panner(Coordinate=@n16, Time=@n17) {speedY=1}) {g=true}
   B=ScalarParameter() {defaultValue=10, parameterName=Scalines Amount}
  )
  NamedRerouteDeclaration(None=SmoothStep(Min=Saturate(None=@n15), Max=@n14, Value=Frac(None=@n18))) {name=Scalines, nodeColor=<rgba 0 0 0 1>}
  n19 = ComponentMask(None=TextureCoordinate()) {g=true}
  n20 = Frac(None=Multiply(A=Time(), B=ScalarParameter() {defaultValue=0.2, parameterName=Speed}))
  OneMinus(None=Sine(None=Add(A=@n19, B=@n20)) {period=0.1})
 ]
}
