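# Translucent UI material for a decorative side element of the Jethro Lager boss HUD.
# It samples the side-elements texture with centre-scaled UVs and decodes the green channel through a channel-packed distance field function.
# The resulting fill is multiplied by an Opacity parameter, and a DecorationColor parameter supplies a tint colour.
# ShowRightPartDecoration blends green with itself, so it has no visible effect.
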Material /JethroLager/NPCs/Bosses/UI/M_UI_JethroLager_Decoration
graph {
 properties {bAutomaticallySetUsageInEditor=false, blendMode=BLEND_Translucent, materialDomain=MD_UI, usedShadingModels=Default Lit}
 nodes [
  n1 = TextureSample(UVs=MaterialFunctionCall(UVs=TextureCoordinate(), Texture Scale=Constant() {r=2}) {materialFunction=&ScaleUVsByCenter}.UVs) {texture=/JethroLager/NPCs/Bosses/UI/Textures/T_UI_Jethro_HUD_SIdeElements}
  n2 = ScalarParameter() {parameterName=ShowRightPartDecoration}
  Multiply(
   A=MaterialFunctionCall!(RGBA Channels=LinearInterpolate(A=@n1.G, B=@n1.G, Alpha=@n2), Glow Min, Glow Max, Outline Thickness) {materialFunction=/FortUILibrary/MaterialFunctions/Basics/MF_UI_ChannelPackedDistanceField}.Fill
   B=ScalarParameter() {defaultValue=1, parameterName=Opacity}
  )
  VectorParameter() {parameterName=DecorationColor, defaultValue=<rgba 1 1 1 1>}
 ]
}
