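# Translucent UI material for text in the Jethro Lager boss HUD.
# The red channel of the HUD side-elements texture feeds MF_UI_ChannelPackedDistanceField, treating it as a distance field.
# The function's Fill output is multiplied by an Opacity parameter, alongside a TextColor parameter that defaults to white.
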Material /JethroLager/NPCs/Bosses/UI/M_UI_JethroLager_Text
graph {
 properties {bAutomaticallySetUsageInEditor=false, blendMode=BLEND_Translucent, materialDomain=MD_UI, usedShadingModels=Default Lit}
 nodes [
  Multiply(
   A=MaterialFunctionCall!(
    RGBA Channels=TextureSample() {texture=/JethroLager/NPCs/Bosses/UI/Textures/T_UI_Jethro_HUD_SIdeElements}.R
    Glow Min
    Glow Max
    Outline Thickness
   ) {materialFunction=/FortUILibrary/MaterialFunctions/Basics/MF_UI_ChannelPackedDistanceField}.Fill
   B=ScalarParameter() {defaultValue=0.5, parameterName=Opacity}
  )
  VectorParameter() {parameterName=TextColor, defaultValue=<rgba 1 1 1 1>}
 ]
}
